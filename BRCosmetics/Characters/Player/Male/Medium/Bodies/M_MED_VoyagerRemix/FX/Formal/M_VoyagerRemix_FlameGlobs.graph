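# Translucent, two-sided particle material for the flame globs FX on the VoyagerRemix character body.
# Particle color red and green thresholds pick one of three palettes, blended through a rotating camera-vector triplanar nebula texture plus an edge tint.
# Dynamic parameters drive sine-distorted, curve-stretched UVs, whose sphere masks shape each glob.
# Depth fade softens it and a camera-distance offset shrinks nearby particles, with simpler mobile paths for texture, distortion, fade and offset.
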Material /BRCosmetics/Characters/Player/Male/Medium/Bodies/M_MED_VoyagerRemix/FX/Formal/M_VoyagerRemix_FlameGlobs
graph {
 properties {
  bUsedWithNiagaraSprites=true
  bUsedWithParticleSprites=true
  blendMode=BLEND_Translucent
  pixelDepthOffsetMode=PDOM_Legacy
  twoSided=true
  usedShadingModels=Default Lit
 }
 nodes [
  n1 = ParticleColor()
  n2 = VectorParameter() {parameterName=R_Color_baseColor, defaultValue=<rgba 0.2016 0 0.0097 1>}
  n3 = VectorParameter() {parameterName=G_Color_baseColor, defaultValue=<rgba 0.05 0.25 0 1>}
  n4 = If(
   A=ComponentMask(None=@n1.RGB) {r=true}
   B=Constant() {r=1}
   "A > B"=@n2.RGB
   "A == B"=@n2.RGB
   "A < B"=If(
    A=ComponentMask(None=@n1.RGB) {g=true}
    B=Constant() {r=1}
    "A > B"=@n3.RGB
    "A == B"=@n3.RGB
    "A < B"=VectorParameter() {parameterName=B_Color_baseColor, defaultValue=<rgba 0 0.0742 0.0782 1>}.RGB
   )
  )
  n5 = TextureObject() {texture=/Game/Characters/Player/Male/Medium/Bodies/M_Med_Soldier_04/Skins/Galaxy/Textures/T_FN_Nebula}
  n6 = MaterialFunctionCall!(
   CameraVector
   Rotation Speed=Multiply(A=Time(), B=ScalarParameter() {defaultValue=0.05, group=Galaxy, parameterName=Nebula Rotation speed})
   Texture=@n5
   Tiling=ScalarParameter() {defaultValue=5, parameterName=Nebula_Base_tile}
   Offset
   AxisFadeContrast
  ) {materialFunction=/Game/Weapons/WeaponSkins/MaterialFunction/TriplanarCameraVector_Rotated}
  n7 = Panner() {speedY=0.03}
  n8 = ShadingPathSwitch(Default=@n6.Result, Deferred=@n6.Result, Mobile=TextureSample(UVs=Multiply(A=@n7) {constB=2}, Tex=@n5).RGB) {desc=Migrated from FeatureLevelSwitch}
  n9 = ComponentMask(None=@n8) {r=true}
  n10 = ParticleColor()
  n11 = VectorParameter() {parameterName=R_Color_insideA, defaultValue=<rgba 0.5365 0.0251 0.3176 1>}
  n12 = VectorParameter() {parameterName=G_Color_insideA, defaultValue=<rgba 0.0232 1 0 1>}
  n13 = If(
   A=ComponentMask(None=@n10.RGB) {r=true}
   B=Constant() {r=1}
   "A > B"=@n11.RGB
   "A == B"=@n11.RGB
   "A < B"=If(
    A=ComponentMask(None=@n10.RGB) {g=true}
    B=Constant() {r=1}
    "A > B"=@n12.RGB
    "A == B"=@n12.RGB
    "A < B"=VectorParameter() {parameterName=B_Color_insideA, defaultValue=<rgba 0.158 0.41 0.41 1>}.RGB
   )
  )
  n14 = ComponentMask(None=@n8) {b=true}
  n15 = ParticleColor()
  n16 = VectorParameter() {parameterName=R_Color_insideB, defaultValue=<rgba 1.2 0.4053 0.125 1>}
  n17 = ComponentMask(None=@n15.RGB) {g=true}
  n18 = Constant() {r=1}
  n19 = VectorParameter() {parameterName=G_Color_insideB, defaultValue=<rgba 1 0 0 1>}
  n20 = VectorParameter() {parameterName=B_Color_insideB, defaultValue=<rgba 0.3613 1 1 1>}
  n21 = LinearInterpolate(
   B=If(
    A=ComponentMask(None=@n15.RGB) {r=true}
    B=Constant() {r=1}
    "A > B"=@n16.RGB
    "A == B"=@n16.RGB
    "A < B"=If(A=@n17, B=@n18, "A > B"=@n19.RGB, "A == B"=@n19.RGB, "A < B"=@n20.RGB)
   )
   Alpha=@n14
  )
  n22 = ParticleColor()
  n23 = ComponentMask(None=@n22.RGB) {r=true}
  n24 = Constant() {r=1}
  n25 = VectorParameter() {parameterName=R_Color_edge, defaultValue=<rgba 1 0.0742 0.3712 1>}
  n26 = VectorParameter() {parameterName=G_Color_edge, defaultValue=<rgba 0.0844 0.7084 1 1>}
  n27 = If(
   A=ComponentMask(None=@n22.RGB) {g=true}
   B=Constant() {r=1}
   "A > B"=@n26.RGB
   "A == B"=@n26.RGB
   "A < B"=VectorParameter() {parameterName=B_Color_edge, defaultValue=<rgba 0.0612 0.3763 1 1>}.RGB
  )
  n28 = Constant2Vector()
  n29 = DynamicParameter() {paramNames ["tail_strecth(0-1)", "maskRadius(0-1)", "distortion_bias(0-1)", tail_distortion]}
  n30 = Multiply(A=@n29."distortion_bias(0-1)")
  n31 = SphereMask(
   A=WorldPosition().XYZ
   B=Add(A=ActorPositionWS(), B=AppendVector(A=@n28, B=@n30))
   Radius=ScalarParameter() {defaultValue=150, parameterName=Radius}
   Hardness=ScalarParameter() {parameterName=Hardness}
  ) {attenuationRadius=20, hardnessPercent=1}
  n32 = ScalarParameter() {defaultValue=15, parameterName=Loop}
  n33 = Multiply(
   A=Sine(None=Multiply(A=@n31, B=@n32))
   B=DynamicParameter() {paramNames ["tail_strecth(0-1)", "maskRadius(0-1)", "distortion_bias(0-1)", tail_distortion]}.tail_distortion
  ) {constB=0.1}
  n34 = TextureCoordinate()
  n35 = Time()
  n36 = ComponentMask(None=Panner(Coordinate=@n34, Time=Add(A=@n29."distortion_bias(0-1)", B=@n35)) {speedY=-0.6}) {g=true}
  n37 = ShadingPathSwitch(
   Default=@n33
   Deferred=@n33
   Mobile=Multiply(
    A=Sine(None=@n36) {period=2}
    B=DynamicParameter() {paramNames ["tail_strecth(0-1)", "maskRadius(0-1)", "distortion_bias(0-1)", tail_distortion]}.tail_distortion
   ) {constB=0.1}
  ) {desc=Migrated from FeatureLevelSwitch}
  n38 = ComponentMask(None=@n34) {g=true}
  n39 = ComponentMask(None=@n34) {g=true}
  n40 = CurveAtlasRowParameter(CurveTime=@n39) {
   atlas=&AthenaCurveAtlas01
   curve=/Game/Characters/Player/Male/Medium/Bodies/M_MED_VoyagerRemix/FX/Formal/Curve_VoyagerRemix2
   defaultValue=119
  }
  n41 = AppendVector(
   A=Add(A=Multiply(A=@n37, B=@n38), B=ComponentMask(None=@n34) {r=true})
   B=LinearInterpolate(
    A=LinearInterpolate(Alpha=@n39) {constA=0.15}
    B=@n40.A
    Alpha=DynamicParameter() {paramNames ["tail_strecth(0-1)", "maskRadius(0-1)", "distortion_bias(0-1)", tail_distortion]}."tail_strecth(0-1)"
   )
  )
  n42 = Multiply(
   A=Constant() {r=0.249}
   B=DynamicParameter() {paramNames ["tail_strecth(0-1)", "maskRadius(0-1)", "distortion_bias(0-1)", tail_distortion]}."maskRadius(0-1)"
  )
  n43 = AppendVector(A=Constant() {r=0.5}, B=Add(A=@n42) {constB=0.15})
  n44 = SphereMask(A=@n41, B=@n43) {attenuationRadius=0.2, hardnessPercent=0}
  n45 = Add(
   A=Add(A=LinearInterpolate(A=@n4, B=@n13, Alpha=@n9), B=Multiply(A=@n14, B=@n21))
   B=LinearInterpolate(B=If(A=@n23, B=@n24, "A > B"=@n25.RGB, "A == B"=@n25.RGB, "A < B"=@n27), Alpha=@n44)
  ) {constB=0}
  n46 = ParticleColor()
  MaterialFunctionCall(Emissive=Add(A=VectorParameter() {parameterName=BaseColorAdd}.RGB, B=Multiply(A=@n45, B=@n46.A))) {materialFunction=&MF_FX_EmissiveAdjustments}
  n47 = CurveAtlasRowParameter(CurveTime=@n44) {atlas=&AthenaCurveAtlas01, curve=/Game/Athena/LinearColorCurves/Curve_Birthday2019Pickaxe, defaultValue=112}
  Reroute(None=@n47.G)
  n48 = Multiply(B=SphereMask(A=@n41, B=@n43, Radius=@n42) {attenuationRadius=0.2, hardnessPercent=70}) {constA=1}
  ShadingPathSwitch(Default=Multiply(A=@n48, B=DepthFade() {fadeDistanceDefault=5}), Mobile=@n48) {desc=Migrated from FeatureLevelSwitch}
  n49 = ParticlePositionWS()
  n50 = WorldPosition()
  n51 = Divide(A=Distance(A=CameraPositionWS(), B=@n49), B=ScalarParameter() {defaultValue=50, parameterName=NearShrinkDist})
  ShadingPathSwitch(Default=LinearInterpolate(A=Subtract(A=@n49, B=@n50.XYZ), Alpha=Saturate(None=@n51)), Mobile=Constant()) {desc=Migrated from FeatureLevelSwitch}
 ]
}
