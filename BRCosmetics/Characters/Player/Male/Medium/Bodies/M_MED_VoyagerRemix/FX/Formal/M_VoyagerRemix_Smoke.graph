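# Translucent unlit particle sprite material for smoke FX on the VoyagerRemix player body cosmetic.
# Two panning samples of a tiling smoke puff texture, offset per particle by a dynamic parameter, are summed, contrasted and masked by a sine gradient along U.
# The resulting mask, scaled by particle alpha, feeds a depth fade.
# It also blends blue tints, varied by a random-color dynamic parameter and multiplied by particle color.
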Material /BRCosmetics/Characters/Player/Male/Medium/Bodies/M_MED_VoyagerRemix/FX/Formal/M_VoyagerRemix_Smoke
graph {
 properties {
  bUsedWithParticleSprites=true
  blendMode=BLEND_Translucent
  pixelDepthOffsetMode=PDOM_Legacy
  shadingModel=MSM_Unlit
  usedShadingModels=Unlit
 }
 nodes [
  n1 = ParticleColor()
  n2 = TextureCoordinate()
  n3 = MaterialFunctionCall(Float2=Sine(None=Divide(A=@n2))) {materialFunction=&BreakOutFloat2Components}
  n4 = TextureCoordinate()
  n5 = ScalarParameter() {defaultValue=0.05, parameterName=U1}
  n6 = ScalarParameter() {defaultValue=0.05, parameterName=V1}
  n7 = Multiply(A=Time(), B=ScalarParameter() {defaultValue=1, parameterName=Panner Speed})
  n8 = Panner(Coordinate=Multiply(A=@n4, B=AppendVector(A=@n5, B=@n6)), Time=@n7) {bFractionalPart=true, speedY=0.01}
  n9 = DynamicParameter() {paramNames ["UV_offset(0-1)", "Color_random(0-1)", Param3, Param4]}
  n10 = ScalarParameter() {defaultValue=0.05, parameterName=U2}
  n11 = ScalarParameter() {defaultValue=0.05, parameterName=V2}
  n12 = Panner(Coordinate=Multiply(A=@n4, B=AppendVector(A=@n10, B=@n11)), Time=@n7) {bFractionalPart=true, speedY=0.02}
  n13 = Add(
   A=TextureSample(UVs=Add(A=@n8, B=@n9."UV_offset(0-1)")) {texture=/Game/Global/Textures/Default/Tileable_Patterns/T_Smoke_Puff_Tile_Up}.RGB
   B=TextureSample(UVs=Add(A=@n9."UV_offset(0-1)", B=@n12)) {texture=&T_Smoke_Puff_Tile_Up}.RGB
  )
  n14 = Multiply(
   A=Power(Base=Multiply(A=@n3.R, B=@n3.R), Exp=ScalarParameter() {defaultValue=2, parameterName=Gradient Exponent})
   B=Power(Base=Divide(A=@n13), Exp=ScalarParameter() {defaultValue=3, parameterName=Cloud Exponent}) {constExponent=3}
  )
  DepthFade(Opacity=Multiply(A=@n1.A, B=@n14), FadeDistance=ScalarParameter() {defaultValue=3, parameterName=Depth Fade Distance})
  MaterialFunctionCall(Emissive=@n1.RGB) {materialFunction=&MF_FX_EmissiveAdjustments}
  n15 = VectorParameter() {defaultValue=<rgba 0 0.0325 0.4375 1>}
  n16 = VectorParameter() {parameterName=Param_1, defaultValue=<rgba 0.4811 0.7173 0.9427 1>}
  n17 = DynamicParameter() {paramNames ["UV_offset(0-1)", "Color_random(0-1)", Param3, Param4]}
  n18 = VectorParameter() {parameterName=Param_2, defaultValue=<rgba 0.4287 0.5422 0.7552 1>}
  n19 = VectorParameter() {parameterName=Param_3, defaultValue=<rgba 0.5625 0.7145 1 1>}
  Multiply(
   A=LinearInterpolate(
    A=LinearInterpolate(A=@n15.RGB, B=@n16.RGB, Alpha=@n17."Color_random(0-1)")
    B=LinearInterpolate(A=@n18.RGB, B=@n19.RGB, Alpha=@n17."Color_random(0-1)")
    Alpha=@n14
   )
   B=@n1.RGB
  )
 ]
}
